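# Physics asset for the NeonCat pickaxe skeletal mesh, a cosmetic two-handed melee weapon.
# A single body on the Melee_TwoHanded_Handle bone holds one large capsule, offset and rotated from the bone origin, as the only collision shape.
# It has no constraints, so the asset defines no jointed or articulated parts.
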PhysicsAsset /BRCosmetics/Weapons/FORT_Melee/Pickaxe_NeonCat/Meshes/DEMO/SK_Pickaxe_NeonCat_PhysicsAsset
bodies {Melee_TwoHanded_Handle [{shapeType=Capsule, length=67.3702, radius=35.3351, center=<0.365 9.2813 24.9177>, rotation=<rot -3.162 -90 -0.0029>}]}
constraints []
